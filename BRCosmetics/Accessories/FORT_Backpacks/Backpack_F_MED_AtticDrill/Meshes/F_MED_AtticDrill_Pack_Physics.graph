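# Physics asset for the AtticDrill backpack mesh, giving capsule bodies to its left and right chains, bottom strings and bottom bows.
# Each side has a six-link chain and a two-segment string, constrained link to link back to small MSTR anchor bodies.
# Joints carry small swing limit angles, and every chain joint locks twist.
# The bow bodies have no constraints, and the right bow uses two capsules.
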PhysicsAsset /BRCosmetics/Accessories/FORT_Backpacks/Backpack_F_MED_AtticDrill/Meshes/F_MED_AtticDrill_Pack_Physics
bodies {
 MSTR_bottomString_l [{shapeType=Capsule, length=0.505, radius=0.505, shapeName=MSTR_bottomString_l_capsule, rotation=<rot 0 -90 0>}]
 MSTR_bottomString_r [{shapeType=Capsule, length=0.505, radius=0.505, shapeName=MSTR_bottomString_r_capsule, rotation=<rot 0 -90 0>}]
 MSTR_chain_l [{shapeType=Capsule, length=0.505, radius=0.505, shapeName=MSTR_chain_l_capsule, rotation=<rot 0 -90 0>}]
 MSTR_chain_r [{shapeType=Capsule, length=0.505, radius=0.505, shapeName=MSTR_chain_r_capsule, rotation=<rot 0 -90 0>}]
 dyn_bottomBow_l [{
   shapeType=Capsule
   length=12.019
   radius=3.0386
   shapeName=dyn_bottomBow_l_capsule
   center=<12.057 -1.8649 2.1095>
   rotation=<rot -0.2693 -94.7294 -98.1309>
  }]
 dyn_bottomBow_r [
  {
   shapeType=Capsule
   length=8.6659
   radius=3.12
   shapeName=dyn_bottomBow_r_capsule
   center=<11.9572 -3.4617 -3.4769>
   rotation=<rot -1.4632 75.1827 -104.0826>
  }
  {
   shapeType=Capsule
   length=3.3567
   radius=1.7916
   shapeName=dyn_bottomBow_r_capsule
   center=<6.8639 -1.1068 -2.7784>
   rotation=<rot -56.8461 -179.9998 169.9472>
  }
 ]
 dyn_bottomStringEnd_l [{
   shapeType=Capsule
   length=5.6371
   radius=0.8013
   shapeName=dyn_bottomStringEnd_l_capsule
   center=<3.7505 -0.1305 -0.3638>
   rotation=<rot 0.1762 86.7865 86.2291>
  }]
 dyn_bottomStringEnd_r [{
   shapeType=Capsule
   length=5.6371
   radius=0.8013
   shapeName=dyn_bottomStringEnd_r_capsule
   center=<4.5902 -0.4306 0.4324>
   rotation=<rot 0 84.1181 -83.2504>
  }]
 dyn_bottomString_l [{
   shapeType=Capsule
   length=6.1925
   radius=0.7131
   shapeName=dyn_bottomString_l_capsule
   center=<6.7245 0.3669 0.1377>
   rotation=<rot -0.1674 85.6374 86.1562>
  }]
 dyn_bottomString_r [{
   shapeType=Capsule
   length=6.1925
   radius=0.7131
   shapeName=dyn_bottomString_r_capsule
   center=<7.2872 -0.2141 -0.0774>
   rotation=<rot 0.0327 87.268 -87.4762>
  }]
 dyn_chain_01_l [{
   shapeType=Capsule
   length=4.3255
   radius=0.7096
   shapeName=dyn_chain_01_l_capsule
   center=<2.5854 -0.2659 0.0705>
   rotation=<rot 0 -88.7687 -89.6427>
  }]
 dyn_chain_01_r [{
   shapeType=Capsule
   length=4.3255
   radius=0.7096
   shapeName=dyn_chain_01_r_capsule
   center=<2.742 -0.191 -0.1037>
   rotation=<rot 0 -88.7687 90.3573>
  }]
 dyn_chain_02_l [{
   shapeType=Capsule
   length=3.2006
   radius=0.7188
   shapeName=dyn_chain_02_l_capsule
   center=<2.3406 -0.1269 0.0427>
   rotation=<rot 0 -88.1963 -89.6511>
  }]
 dyn_chain_02_r [{
   shapeType=Capsule
   length=3.2006
   radius=0.7188
   shapeName=dyn_chain_02_r_capsule
   center=<2.5885 0.2821 -0.0728>
   rotation=<rot 0 -88.1963 90.3489>
  }]
 dyn_chain_03_l [{
   shapeType=Capsule
   length=4.0838
   radius=0.7592
   shapeName=dyn_chain_03_l_capsule
   center=<1.929 0.0105 0.0164>
   rotation=<rot 0 92.3321 -89.8536>
  }]
 dyn_chain_03_r [{
   shapeType=Capsule
   length=4.0838
   radius=0.7592
   shapeName=dyn_chain_03_r_capsule
   center=<2.1168 0.1543 -0.006>
   rotation=<rot 0 92.3321 90.1464>
  }]
 dyn_chain_04_l [{
   shapeType=Capsule
   length=2.5081
   radius=1.7445
   shapeName=dyn_chain_04_l_capsule
   center=<3.1379 0.2032 -0.1144>
   rotation=<rot 0 -89.2576 -88.566>
  }]
 dyn_chain_04_r [{
   shapeType=Capsule
   length=2.5081
   radius=1.7445
   shapeName=dyn_chain_04_r_capsule
   center=<2.6202 -0.1286 0.1167>
   rotation=<rot 0 -89.2579 90.8922>
  }]
 dyn_chain_05_l [{
   shapeType=Capsule
   length=2.3457
   radius=2.3284
   shapeName=dyn_chain_05_l_capsule
   center=<3.3565 0.0448 -0.2515>
   rotation=<rot 63.1249 93.4139 -90.3156>
  }]
 dyn_chain_05_r [{
   shapeType=Capsule
   length=2.3456
   radius=2.3284
   shapeName=dyn_chain_05_r_capsule
   center=<3.3603 -0.0132 0.4174>
   rotation=<rot 63.1619 -86.6491 -90.3161>
  }]
 dyn_chain_06_l [{
   shapeType=Capsule
   length=2.9487
   radius=3.035
   shapeName=dyn_chain_06_l_capsule
   center=<1.3717 0.2903 -0.1531>
   rotation=<rot -79.3002 -82.1926 -95.1577>
  }]
 dyn_chain_06_r [{
   shapeType=Capsule
   length=2.9487
   radius=3.0349
   shapeName=dyn_chain_06_r_capsule
   center=<1.3715 0.2911 0.2015>
   rotation=<rot -79.6337 97.393 85.2497>
  }]
}
constraints [
 {bone1Name=dyn_bottomStringEnd_l, bone2Name=dyn_bottomString_l, swing1LimitDegrees=15, swing2LimitDegrees=5, twistLimitDegrees=15}
 {bone1Name=dyn_bottomStringEnd_r, bone2Name=dyn_bottomString_r, swing1LimitDegrees=15, swing2LimitDegrees=8, twistLimitDegrees=15}
 {bone1Name=dyn_chain_01_r, bone2Name=MSTR_chain_r, swing1LimitDegrees=7, swing2LimitDegrees=10, twistLimitDegrees=45, twistMotion=Locked}
 {bone1Name=dyn_chain_02_r, bone2Name=dyn_chain_01_r, swing1LimitDegrees=15, swing2LimitDegrees=17, twistLimitDegrees=45, twistMotion=Locked}
 {bone1Name=dyn_chain_03_r, bone2Name=dyn_chain_02_r, swing1LimitDegrees=11, swing2LimitDegrees=16, twistLimitDegrees=45, twistMotion=Locked}
 {bone1Name=dyn_chain_01_l, bone2Name=MSTR_chain_l, swing1LimitDegrees=7, swing2LimitDegrees=10, twistLimitDegrees=45, twistMotion=Locked}
 {bone1Name=dyn_chain_02_l, bone2Name=dyn_chain_01_l, swing1LimitDegrees=10, swing2LimitDegrees=19, twistLimitDegrees=45, twistMotion=Locked}
 {bone1Name=dyn_chain_03_l, bone2Name=dyn_chain_02_l, swing1LimitDegrees=7, swing2LimitDegrees=8, twistLimitDegrees=45, twistMotion=Locked}
 {bone1Name=dyn_chain_04_l, bone2Name=dyn_chain_03_l, swing1LimitDegrees=12, swing2LimitDegrees=20, twistLimitDegrees=45, twistMotion=Locked}
 {bone1Name=dyn_chain_04_r, bone2Name=dyn_chain_03_r, swing1LimitDegrees=17, swing2LimitDegrees=20, twistLimitDegrees=45, twistMotion=Locked}
 {bone1Name=dyn_bottomString_r, bone2Name=MSTR_bottomString_r, swing1LimitDegrees=5, swing2LimitDegrees=5, twistLimitDegrees=15}
 {bone1Name=dyn_bottomString_l, bone2Name=MSTR_bottomString_l, swing1LimitDegrees=5, swing2LimitDegrees=5, twistLimitDegrees=15}
 {bone1Name=dyn_chain_05_r, bone2Name=dyn_chain_04_r, swing1LimitDegrees=10, swing2LimitDegrees=10, twistLimitDegrees=45, twistMotion=Locked}
 {bone1Name=dyn_chain_06_r, bone2Name=dyn_chain_05_r, swing1LimitDegrees=10, swing2LimitDegrees=10, twistLimitDegrees=45, twistMotion=Locked}
 {bone1Name=dyn_chain_05_l, bone2Name=dyn_chain_04_l, swing1LimitDegrees=10, swing2LimitDegrees=10, twistLimitDegrees=45, twistMotion=Locked}
 {bone1Name=dyn_chain_06_l, bone2Name=dyn_chain_05_l, swing1LimitDegrees=10, swing2LimitDegrees=10, twistLimitDegrees=45, twistMotion=Locked}
]
