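# Animation Blueprint for the AtticDrill backpack cosmetic, adding secondary motion to its bows, strings, chains and bamboo stick.
# A cached bamboo-stick simulation, disabled while grinding or skydive-floating, feeds two branches selected by bIsInFrontEnd.
# Front end uses AnimDynamics throughout; otherwise RigidBody plus AnimDynamics runs, with bone nudges for moving, skydiving and DBNO.
# Final bone offsets shift and roll the pack from spine yaw computed in the EventGraph.
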AnimBlueprint /BRCosmetics/Accessories/FORT_Backpacks/Backpack_F_MED_AtticDrill/Meshes/F_MED_AtticDrill_Pack_AnimBP
target_skeleton=&/F_MED_AtticDrill_Pack_Skeleton
graphs [
 AnimGraph {nodes [
   n1 = "Transform (Modify) Bone - Bone: dyn_chain_04_l"!(ComponentPose, bAlphaBoolEnabled:bool=True, Translation="0,0,0", Rotation="0,8,0", Scale="1,1,1") -> (Pose) {
    alphaInputType=Bool
    rotationMode=BMM_Additive
    rotationSpace=BCS_BoneSpace
    translationMode=BMM_Additive
    translationSpace=BCS_BoneSpace
    alphaBoolBlend {blendInTime=0.2, blendOutTime=0.3}
    boneToModify {boneName=dyn_chain_04_l}
    rotation {yaw=8}
   }
   "Transform (Modify) Bone - Bone: dyn_chain_04_r"!(ComponentPose=@n1, bAlphaBoolEnabled:bool=True, Translation="0,0,0", Rotation="0,8,0", Scale="1,1,1") -> (Pose) {
    alphaInputType=Bool
    rotationMode=BMM_Additive
    rotationSpace=BCS_BoneSpace
    translationMode=BMM_Additive
    translationSpace=BCS_BoneSpace
    alphaBoolBlend {blendInTime=0.2, blendOutTime=0.3}
    boneToModify {boneName=dyn_chain_04_r}
    rotation {yaw=8}
   }
   n2 = Anim Dynamics - Bone: dyn_topBow_r!(ComponentPose=Local To Component(LocalPose=Use cached pose 'chainRBAN'()), Alpha=1, GravityOverride:Vector="0,0,-980", ExternalForce="0,0,0") -> (Pose) {
    angularDampingOverride=0.999
    angularSpringConstant=12000
    bAngularSpring=true
    bOverrideAngularDamping=true
    bUseGravityOverride=true
    gravityScale=0.35
    simulationSpace=BoneRelative
    boundBone {boneName=dyn_topBow_r}
    componentAppliedLinearAccClamp=<3000 3000 3000>
    componentLinearAccScale=<0.3 0.3 0.6>
    componentLinearVelScale=<0.3 0.3 0.6>
    gravityOverride {z=-980}
    physicsBodyDefinitions [{
      boundBone {boneName=dyn_topBow_r}
      boxExtents=<35 30 30>
      constraintSetup {angularLimitsMax=<0 10 0>, angularLimitsMin=<0 -30 0>, angularTarget=<1 0 0>}
      localJointOffset=<0.2 0.2 -1.3>
     }]
    relativeSpaceBone {boneName=pack}
    simSpaceSettings {simSpaceAngularAlpha=0.001}
   }
   n3 = Anim Dynamics - Bone: dyn_bottomBow_r!(ComponentPose=@n2, Alpha=1, GravityOverride:Vector="0,0,-980", ExternalForce="0,0,0") -> (Pose) {
    angularDampingOverride=0.999
    angularSpringConstant=4000
    bAngularSpring=true
    bOverrideAngularDamping=true
    bUseGravityOverride=true
    gravityScale=0.35
    simulationSpace=BoneRelative
    boundBone {boneName=dyn_bottomBow_r}
    componentAppliedLinearAccClamp=<3000 3000 3000>
    componentLinearAccScale=<0.3 0.3 0.5>
    componentLinearVelScale=<0.3 0.3 0.5>
    gravityOverride {z=-980}
    physicsBodyDefinitions [{
      boundBone {boneName=dyn_bottomBow_r}
      boxExtents=<20 20 20>
      constraintSetup {angularLimitsMax=<0 4 0>, angularLimitsMin=<0 -20 0>, angularTarget=<1 0 0>}
      localJointOffset=<0.6 0 -0.24>
     }]
    relativeSpaceBone {boneName=pack}
    simSpaceSettings {simSpaceAngularAlpha=0.001}
   }
   n4 = Anim Dynamics - Bone: dyn_topStrings_r!(ComponentPose=@n3, GravityOverride:Vector="0,0,-980", ExternalForce="0,0,0") -> (Pose) {
    alpha=0.7
    angularDampingOverride=0.999
    angularSpringConstant=3000
    bAngularSpring=true
    bOverrideAngularDamping=true
    bUseGravityOverride=true
    gravityScale=0.35
    simulationSpace=BoneRelative
    boundBone {boneName=dyn_topStrings_r}
    componentAppliedLinearAccClamp=<3000 3000 3000>
    componentLinearAccScale=<0.3 0.3 0.5>
    componentLinearVelScale=<0.3 0.3 0.5>
    gravityOverride {z=-980}
    physicsBodyDefinitions [{
      boundBone {boneName=dyn_topStrings_r}
      boxExtents=<20 20 20>
      constraintSetup {angularLimitsMax=<0 10 0>, angularLimitsMin=<0 -20 0>, angularTarget=<1 0 0>}
      localJointOffset=<0.2 0 0.3>
     }]
    relativeSpaceBone {boneName=pack}
    simSpaceSettings {simSpaceAngularAlpha=0.2}
   }
   n5 = Anim Dynamics - Bone: dyn_topStringsEnd_r!(ComponentPose=@n4, GravityOverride:Vector="0,0,-980", ExternalForce="0,0,0") -> (Pose) {
    alpha=0.7
    angularDampingOverride=0.999
    angularSpringConstant=2000
    bAngularSpring=true
    bOverrideAngularDamping=true
    bUseGravityOverride=true
    gravityScale=0.35
    simulationSpace=BoneRelative
    boundBone {boneName=dyn_topStringsEnd_r}
    componentAppliedLinearAccClamp=<3000 3000 3000>
    componentLinearAccScale=<0.05 0.03 0.5>
    componentLinearVelScale=<0.05 0.01 0.5>
    gravityOverride {z=-980}
    physicsBodyDefinitions [{
      boundBone {boneName=dyn_topStringsEnd_r}
      boxExtents=<30 30 30>
      constraintSetup {angularLimitsMax=<0 5 2>, angularLimitsMin=<0 -5 0>, angularTarget=<1 0 0>}
      localJointOffset=<1.5 -1 0.6>
     }]
    relativeSpaceBone {boneName=pack}
    simSpaceSettings {simSpaceAngularAlpha=0.2}
   }
   n6 = Anim Dynamics - Bone: dyn_topStringsSml_r!(ComponentPose=@n5, GravityOverride:Vector="0,0,-980", ExternalForce="0,0,0") -> (Pose) {
    alpha=0.7
    angularDampingOverride=0.999
    angularSpringConstant=3200
    bAngularSpring=true
    bOverrideAngularDamping=true
    bUseGravityOverride=true
    gravityScale=0.35
    simulationSpace=BoneRelative
    boundBone {boneName=dyn_topStringsSml_r}
    componentAppliedLinearAccClamp=<3000 3000 3000>
    componentLinearAccScale=<0.3 0.3 0.5>
    componentLinearVelScale=<0.3 0.3 0.5>
    gravityOverride {z=-980}
    physicsBodyDefinitions [{
      boundBone {boneName=dyn_topStringsSml_r}
      boxExtents=<20 20 20>
      constraintSetup {angularLimitsMax=<0 10 5>, angularLimitsMin=<0 -20 0>, angularTarget=<1 0 0>}
      localJointOffset=<0.3 -1 0.3>
     }]
    relativeSpaceBone {boneName=pack}
    simSpaceSettings {simSpaceAngularAlpha=0.2}
   }
   n7 = Anim Dynamics - Bone: dyn_topBow_l!(ComponentPose=@n6, Alpha=1, GravityOverride:Vector="0,0,-980", ExternalForce="0,0,0") -> (Pose) {
    angularDampingOverride=0.999
    angularSpringConstant=12000
    bAngularSpring=true
    bOverrideAngularDamping=true
    bUseGravityOverride=true
    gravityScale=0.35
    simulationSpace=BoneRelative
    boundBone {boneName=dyn_topBow_l}
    componentAppliedLinearAccClamp=<3000 3000 3000>
    componentLinearAccScale=<0.3 0.3 0.6>
    componentLinearVelScale=<0.3 0.3 0.6>
    gravityOverride {z=-980}
    physicsBodyDefinitions [{
      boundBone {boneName=dyn_topBow_l}
      boxExtents=<35 30 30>
      constraintSetup {angularLimitsMax=<0 10 0>, angularLimitsMin=<0 -30 0>, angularTarget=<1 0 0>}
      localJointOffset=<0.2 0.2 -1.3>
     }]
    relativeSpaceBone {boneName=pack}
    simSpaceSettings {simSpaceAngularAlpha=0.001}
   }
   n8 = Anim Dynamics - Bone: dyn_bottomBow_l!(ComponentPose=@n7, Alpha=1, GravityOverride:Vector="0,0,-980", ExternalForce="0,0,0") -> (Pose) {
    angularDampingOverride=0.999
    angularSpringConstant=4000
    bAngularSpring=true
    bOverrideAngularDamping=true
    bUseGravityOverride=true
    gravityScale=0.35
    simulationSpace=BoneRelative
    boundBone {boneName=dyn_bottomBow_l}
    componentAppliedLinearAccClamp=<3000 3000 3000>
    componentLinearAccScale=<0.3 0.3 0.5>
    componentLinearVelScale=<0.3 0.3 0.5>
    gravityOverride {z=-980}
    physicsBodyDefinitions [{
      boundBone {boneName=dyn_bottomBow_l}
      boxExtents=<20 20 20>
      constraintSetup {angularLimitsMax=<0 10 0>, angularLimitsMin=<0 -20 0>, angularTarget=<1 0 0>}
      localJointOffset=<0.35 0 0.2>
     }]
    relativeSpaceBone {boneName=pack}
    simSpaceSettings {simSpaceAngularAlpha=0.001}
   }
   n9 = Anim Dynamics - Bone: dyn_topStrings_l!(ComponentPose=@n8, GravityOverride:Vector="0,0,-980", ExternalForce="0,0,0") -> (Pose) {
    alpha=0.7
    angularDampingOverride=0.95
    angularSpringConstant=900
    bAngularSpring=true
    bOverrideAngularDamping=true
    bUseGravityOverride=true
    gravityScale=0.35
    simulationSpace=BoneRelative
    boundBone {boneName=dyn_topStrings_l}
    componentAppliedLinearAccClamp=<3000 3000 3000>
    componentLinearAccScale=<0.3 0.3 0.4>
    componentLinearVelScale=<0.3 0.3 0.4>
    gravityOverride {z=-980}
    physicsBodyDefinitions [{
      boundBone {boneName=dyn_topStrings_l}
      boxExtents=<20 20 20>
      constraintSetup {angularLimitsMax=<0 10 0>, angularLimitsMin=<0 -20 0>, angularTarget=<1 0 0>}
      localJointOffset=<0.1 0 0.3>
     }]
    relativeSpaceBone {boneName=pack}
    simSpaceSettings {simSpaceAngularAlpha=0.2}
   }
   n10 = Anim Dynamics - Bone: dyn_topStringsEnd_l!(ComponentPose=@n9, GravityOverride:Vector="0,0,-980", ExternalForce="0,0,0") -> (Pose) {
    alpha=0.7
    angularDampingOverride=0.999
    angularSpringConstant=1500
    bAngularSpring=true
    bOverrideAngularDamping=true
    bUseGravityOverride=true
    gravityScale=0.35
    simulationSpace=BoneRelative
    boundBone {boneName=dyn_topStringsEnd_l}
    componentAppliedLinearAccClamp=<3000 3000 3000>
    componentLinearAccScale=<0.05 0.03 0.5>
    componentLinearVelScale=<0.05 0.01 0.5>
    gravityOverride {z=-980}
    physicsBodyDefinitions [{
      boundBone {boneName=dyn_topStringsEnd_l}
      boxExtents=<20 20 20>
      constraintSetup {angularLimitsMax=<0 5 2>, angularLimitsMin=<0 -5 0>, angularTarget=<1 0 0>}
      localJointOffset=<1.5 -1.5 0.6>
     }]
    relativeSpaceBone {boneName=pack}
    simSpaceSettings {simSpaceAngularAlpha=0.2}
   }
   n11 = Anim Dynamics - Bone: dyn_topStringsSml_l!(ComponentPose=@n10, GravityOverride:Vector="0,0,-980", ExternalForce="0,0,0") -> (Pose) {
    alpha=0.7
    angularDampingOverride=0.95
    angularSpringConstant=700
    bAngularSpring=true
    bOverrideAngularDamping=true
    bUseGravityOverride=true
    gravityScale=0.35
    simulationSpace=BoneRelative
    boundBone {boneName=dyn_topStringsSml_l}
    componentAppliedLinearAccClamp=<3000 3000 3000>
    componentLinearAccScale=<0.3 0.3 0.4>
    componentLinearVelScale=<0.3 0.3 0.4>
    gravityOverride {z=-980}
    physicsBodyDefinitions [{
      boundBone {boneName=dyn_topStringsSml_l}
      boxExtents=<20 20 20>
      constraintSetup {angularLimitsMax=<0 10 0>, angularLimitsMin=<0 -20 0>, angularTarget=<1 0 0>}
      localJointOffset=<0.1 0 0.3>
     }]
    relativeSpaceBone {boneName=pack}
    simSpaceSettings {simSpaceAngularAlpha=0.2}
   }
   n12 = "Transform (Modify) Bone - Bone: MSTR_chain_l"!(ComponentPose=@n11, bAlphaBoolEnabled:bool=True, Translation="0,0,0", Rotation="0,-7,0", Scale="1,1,1") -> (Pose) {
    alphaInputType=Bool
    rotationMode=BMM_Additive
    rotationSpace=BCS_BoneSpace
    translationMode=BMM_Additive
    translationSpace=BCS_BoneSpace
    alphaBoolBlend {blendInTime=0.2, blendOutTime=0.3}
    boneToModify {boneName=MSTR_chain_l}
    rotation {yaw=-7}
   }
   n13 = "Transform (Modify) Bone - Bone: MSTR_chain_r"!(ComponentPose=@n12, bAlphaBoolEnabled:bool=True, Translation="0,0,0", Rotation="0,-7,0", Scale="1,1,1") -> (Pose) {
    alphaInputType=Bool
    rotationMode=BMM_Additive
    rotationSpace=BCS_BoneSpace
    translationMode=BMM_Additive
    translationSpace=BCS_BoneSpace
    alphaBoolBlend {blendInTime=0.2, blendOutTime=0.3}
    boneToModify {boneName=MSTR_chain_r}
    rotation {yaw=-7}
   }
   n14 = Anim Dynamics - Bone: dyn_chain_01_l!(ComponentPose=@n13, GravityOverride:Vector="0,0,-980", ExternalForce="0,0,0") -> (Pose) {
    alpha=0.7
    angularDampingOverride=0.95
    angularSpringConstant=4000
    bAngularSpring=true
    bOverrideAngularDamping=true
    bUseGravityOverride=true
    gravityScale=0.35
    simulationSpace=BoneRelative
    boundBone {boneName=dyn_chain_01_l}
    componentAppliedLinearAccClamp=<4000 4000 4000>
    componentLinearAccScale=<0.5 0.5 0.5>
    componentLinearVelScale=<0.5 0.5 0.5>
    gravityOverride {z=-980}
    physicsBodyDefinitions [{
      boundBone {boneName=dyn_chain_01_l}
      boxExtents=<20 20 20>
      constraintSetup {angularLimitsMax=<0 3 5>, angularLimitsMin=<0 0 0>, angularTarget=<1 0 0>}
      localJointOffset=<1.2 0 0.1>
     }]
    relativeSpaceBone {boneName=pack}
    simSpaceSettings {simSpaceAngularAlpha=0.45}
   }
   n15 = Anim Dynamics - Bone: dyn_chain_02_l!(ComponentPose=@n14, GravityOverride:Vector="0,0,-980", ExternalForce="0,0,0") -> (Pose) {
    alpha=0.7
    angularDampingOverride=0.95
    angularSpringConstant=4000
    bAngularSpring=true
    bOverrideAngularDamping=true
    bUseGravityOverride=true
    gravityScale=0.35
    simulationSpace=BoneRelative
    boundBone {boneName=dyn_chain_02_l}
    componentAppliedLinearAccClamp=<4000 4000 4000>
    componentLinearAccScale=<0.5 0.5 0.5>
    componentLinearVelScale=<0.5 0.5 0.5>
    gravityOverride {z=-980}
    physicsBodyDefinitions [{
      boundBone {boneName=dyn_chain_02_l}
      boxExtents=<20 20 20>
      constraintSetup {angularLimitsMax=<0 20 50>, angularLimitsMin=<0 -5 0>, angularTarget=<1 0 0>}
      localJointOffset=<1.5 0 0.1>
     }]
    relativeSpaceBone {boneName=pack}
    simSpaceSettings {simSpaceAngularAlpha=0.45}
   }
   n16 = Anim Dynamics - Bone: dyn_chain_03_l!(ComponentPose=@n15, GravityOverride:Vector="0,0,-980", ExternalForce="0,0,0") -> (Pose) {
    alpha=0.7
    angularDampingOverride=0.95
    angularSpringConstant=4000
    bAngularSpring=true
    bOverrideAngularDamping=true
    bUseGravityOverride=true
    gravityScale=0.35
    simulationSpace=BoneRelative
    boundBone {boneName=dyn_chain_03_l}
    componentAppliedLinearAccClamp=<4000 4000 4000>
    componentLinearAccScale=<0.5 0.5 0.5>
    componentLinearVelScale=<0.5 0.5 0.5>
    gravityOverride {z=-980}
    physicsBodyDefinitions [{
      boundBone {boneName=dyn_chain_03_l}
      boxExtents=<20 20 20>
      constraintSetup {angularLimitsMax=<0 20 50>, angularLimitsMin=<0 -5 0>, angularTarget=<1 0 0>}
      localJointOffset=<1.5 0 0.1>
     }]
    relativeSpaceBone {boneName=pack}
    simSpaceSettings {simSpaceAngularAlpha=0.45}
   }
   n17 = Anim Dynamics - Bone: dyn_chain_04_l!(ComponentPose=@n16, GravityOverride:Vector="0,0,-980", ExternalForce="0,0,0") -> (Pose) {
    alpha=0.7
    angularDampingOverride=0.95
    angularSpringConstant=4000
    bAngularSpring=true
    bOverrideAngularDamping=true
    bUseGravityOverride=true
    gravityScale=0.35
    simulationSpace=BoneRelative
    boundBone {boneName=dyn_chain_04_l}
    componentAppliedLinearAccClamp=<4000 4000 4000>
    componentLinearAccScale=<0.5 0.5 0.5>
    componentLinearVelScale=<0.5 0.5 0.5>
    gravityOverride {z=-980}
    physicsBodyDefinitions [{
      boundBone {boneName=dyn_chain_04_l}
      boxExtents=<20 20 20>
      constraintSetup {angularLimitsMax=<0 20 50>, angularLimitsMin=<0 -5 0>, angularTarget=<1 0 0>}
      localJointOffset=<1.5 0 0.1>
     }]
    relativeSpaceBone {boneName=pack}
    simSpaceSettings {simSpaceAngularAlpha=0.45}
   }
   n18 = Anim Dynamics - Bone: dyn_chain_05_l!(ComponentPose=@n17, GravityOverride:Vector="0,0,-980", ExternalForce="0,0,0") -> (Pose) {
    alpha=0.7
    angularDampingOverride=0.95
    angularSpringConstant=1000
    bAngularSpring=true
    bOverrideAngularDamping=true
    bUseGravityOverride=true
    gravityScale=0.35
    simulationSpace=BoneRelative
    boundBone {boneName=dyn_chain_05_l}
    componentAppliedLinearAccClamp=<4000 4000 4000>
    componentLinearAccScale=<0.5 0.5 0.5>
    componentLinearVelScale=<0.5 0.5 0.5>
    gravityOverride {z=-980}
    physicsBodyDefinitions [{
      boundBone {boneName=dyn_chain_05_l}
      boxExtents=<20 20 20>
      constraintSetup {angularLimitsMax=<0 10 8>, angularLimitsMin=<0 -5 -8>, angularTarget=<1 0 0>}
      localJointOffset=<1.2 0 0.1>
     }]
    relativeSpaceBone {boneName=pack}
    simSpaceSettings {simSpaceAngularAlpha=0.45}
   }
   n19 = Anim Dynamics - Bone: dyn_chain_06_l!(ComponentPose=@n18, GravityOverride:Vector="0,0,-980", ExternalForce="0,0,0") -> (Pose) {
    alpha=0.7
    angularDampingOverride=0.95
    angularSpringConstant=1000
    bAngularSpring=true
    bOverrideAngularDamping=true
    bUseGravityOverride=true
    gravityScale=0.35
    simulationSpace=BoneRelative
    boundBone {boneName=dyn_chain_06_l}
    componentAppliedLinearAccClamp=<4000 4000 4000>
    componentLinearAccScale=<0.5 0.5 0.5>
    componentLinearVelScale=<0.5 0.5 0.5>
    gravityOverride {z=-980}
    physicsBodyDefinitions [{
      boundBone {boneName=dyn_chain_06_l}
      boxExtents=<20 20 20>
      constraintSetup {angularLimitsMax=<0 10 8>, angularLimitsMin=<0 -5 -8>, angularTarget=<1 0 0>}
      localJointOffset=<1.2 0 0.1>
     }]
    relativeSpaceBone {boneName=pack}
    simSpaceSettings {simSpaceAngularAlpha=0.45}
   }
   n20 = Anim Dynamics - Bone: dyn_chain_01_r!(ComponentPose=@n19, GravityOverride:Vector="0,0,-980", ExternalForce="0,0,0") -> (Pose) {
    alpha=0.7
    angularDampingOverride=0.95
    angularSpringConstant=4000
    bAngularSpring=true
    bOverrideAngularDamping=true
    bUseGravityOverride=true
    gravityScale=0.35
    simulationSpace=BoneRelative
    boundBone {boneName=dyn_chain_01_r}
    componentAppliedLinearAccClamp=<4000 4000 4000>
    componentLinearAccScale=<0.5 0.5 0.5>
    componentLinearVelScale=<0.5 0.5 0.5>
    gravityOverride {z=-980}
    physicsBodyDefinitions [{
      boundBone {boneName=dyn_chain_01_r}
      boxExtents=<20 20 20>
      constraintSetup {angularLimitsMax=<0 0 5>, angularLimitsMin=<0 -3 0>, angularTarget=<1 0 0>}
      localJointOffset=<1.2 -0.2 -0.1>
     }]
    relativeSpaceBone {boneName=pack}
    simSpaceSettings {simSpaceAngularAlpha=0.45}
   }
   n21 = Anim Dynamics - Bone: dyn_chain_02_r!(ComponentPose=@n20, GravityOverride:Vector="0,0,-980", ExternalForce="0,0,0") -> (Pose) {
    alpha=0.7
    angularDampingOverride=0.95
    angularSpringConstant=4000
    bAngularSpring=true
    bOverrideAngularDamping=true
    bUseGravityOverride=true
    gravityScale=0.35
    simulationSpace=BoneRelative
    boundBone {boneName=dyn_chain_02_r}
    componentAppliedLinearAccClamp=<4000 4000 4000>
    componentLinearAccScale=<0.5 0.5 0.5>
    componentLinearVelScale=<0.5 0.5 0.5>
    gravityOverride {z=-980}
    physicsBodyDefinitions [{
      boundBone {boneName=dyn_chain_02_r}
      boxExtents=<20 20 20>
      constraintSetup {angularLimitsMax=<0 5 50>, angularLimitsMin=<0 -20 0>, angularTarget=<1 0 0>}
      localJointOffset=<1.5 -0.2 -0.1>
     }]
    relativeSpaceBone {boneName=pack}
    simSpaceSettings {simSpaceAngularAlpha=0.45}
   }
   n22 = Anim Dynamics - Bone: dyn_chain_03_r!(ComponentPose=@n21, GravityOverride:Vector="0,0,-980", ExternalForce="0,0,0") -> (Pose) {
    alpha=0.7
    angularDampingOverride=0.95
    angularSpringConstant=4000
    bAngularSpring=true
    bOverrideAngularDamping=true
    bUseGravityOverride=true
    gravityScale=0.35
    simulationSpace=BoneRelative
    boundBone {boneName=dyn_chain_03_r}
    componentAppliedLinearAccClamp=<4000 4000 4000>
    componentLinearAccScale=<0.5 0.5 0.5>
    componentLinearVelScale=<0.5 0.5 0.5>
    gravityOverride {z=-980}
    physicsBodyDefinitions [{
      boundBone {boneName=dyn_chain_03_r}
      boxExtents=<20 20 20>
      constraintSetup {angularLimitsMax=<0 5 50>, angularLimitsMin=<0 -20 0>, angularTarget=<1 0 0>}
      localJointOffset=<1.5 -0.2 -0.1>
     }]
    relativeSpaceBone {boneName=pack}
    simSpaceSettings {simSpaceAngularAlpha=0.45}
   }
   n23 = Anim Dynamics - Bone: dyn_chain_04_r!(ComponentPose=@n22, GravityOverride:Vector="0,0,-980", ExternalForce="0,0,0") -> (Pose) {
    alpha=0.7
    angularDampingOverride=0.95
    angularSpringConstant=4000
    bAngularSpring=true
    bOverrideAngularDamping=true
    bUseGravityOverride=true
    gravityScale=0.35
    simulationSpace=BoneRelative
    boundBone {boneName=dyn_chain_04_r}
    componentAppliedLinearAccClamp=<4000 4000 4000>
    componentLinearAccScale=<0.5 0.5 0.5>
    componentLinearVelScale=<0.5 0.5 0.5>
    gravityOverride {z=-980}
    physicsBodyDefinitions [{
      boundBone {boneName=dyn_chain_04_r}
      boxExtents=<20 20 20>
      constraintSetup {angularLimitsMax=<0 5 50>, angularLimitsMin=<0 -20 0>, angularTarget=<1 0 0>}
      localJointOffset=<1.5 -0.2 -0.1>
     }]
    relativeSpaceBone {boneName=pack}
    simSpaceSettings {simSpaceAngularAlpha=0.45}
   }
   n24 = Anim Dynamics - Bone: dyn_chain_05_r!(ComponentPose=@n23, GravityOverride:Vector="0,0,-980", ExternalForce="0,0,0") -> (Pose) {
    alpha=0.7
    angularDampingOverride=0.95
    angularSpringConstant=1000
    bAngularSpring=true
    bOverrideAngularDamping=true
    bUseGravityOverride=true
    gravityScale=0.35
    simulationSpace=BoneRelative
    boundBone {boneName=dyn_chain_05_r}
    componentAppliedLinearAccClamp=<4000 4000 4000>
    componentLinearAccScale=<0.5 0.5 0.5>
    componentLinearVelScale=<0.5 0.5 0.5>
    gravityOverride {z=-980}
    physicsBodyDefinitions [{
      boundBone {boneName=dyn_chain_05_r}
      boxExtents=<20 20 20>
      constraintSetup {angularLimitsMax=<0 3 8>, angularLimitsMin=<0 -10 -8>, angularTarget=<1 0 0>}
      localJointOffset=<1.2 0 -0.1>
     }]
    relativeSpaceBone {boneName=pack}
    simSpaceSettings {simSpaceAngularAlpha=0.45}
   }
   n25 = Anim Dynamics - Bone: dyn_chain_06_r!(ComponentPose=@n24, GravityOverride:Vector="0,0,-980", ExternalForce="0,0,0") -> (Pose) {
    alpha=0.7
    angularDampingOverride=0.95
    angularSpringConstant=1000
    bAngularSpring=true
    bOverrideAngularDamping=true
    bUseGravityOverride=true
    gravityScale=0.35
    simulationSpace=BoneRelative
    boundBone {boneName=dyn_chain_06_r}
    componentAppliedLinearAccClamp=<4000 4000 4000>
    componentLinearAccScale=<0.5 0.5 0.5>
    componentLinearVelScale=<0.5 0.5 0.5>
    gravityOverride {z=-980}
    physicsBodyDefinitions [{
      boundBone {boneName=dyn_chain_06_r}
      boxExtents=<20 20 20>
      constraintSetup {angularLimitsMax=<0 3 8>, angularLimitsMin=<0 -10 -8>, angularTarget=<1 0 0>}
      localJointOffset=<1.2 0 -0.1>
     }]
    relativeSpaceBone {boneName=pack}
    simSpaceSettings {simSpaceAngularAlpha=0.45}
   }
   n26 = Component To Local(ComponentPose=@n25)
   n27 = RigidBody!(ComponentPose=Local To Component(LocalPose=Use cached pose 'chainRBAN'()), Alpha=0.9, OverrideWorldGravity:Vector="0,0,0", ExternalForce) -> (Pose) {
    bOverrideWorldGravity=true
    overridePhysicsAsset=&/F_MED_AtticDrill_Pack_Physics
    simulationSpace=BaseBoneSpace
    baseBoneRef {boneName=pack}
    componentAppliedLinearAccClamp=<2000 2000 2000>
    simSpaceSettings {
     dampingAlpha=0.2
     maxAngularAcceleration=20
     maxAngularVelocity=5
     maxLinearAcceleration=1500
     maxLinearVelocity=500
     velocityScaleZ=0.7
     worldAlpha=0.6
     externalLinearDragV {z=0.4}
     externalLinearVelocity=<0.2 0.2 0.2>
    }
   }
   n28 = Anim Dynamics - Bone: dyn_topBow_r(ComponentPose=@n27, Alpha=1, GravityOverride="0,0,-980", ExternalForce="0,0,0") {
    angularDampingOverride=0.999
    angularSpringConstant=3000
    bAngularSpring=true
    bOverrideAngularDamping=true
    bUseGravityOverride=true
    gravityScale=0.35
    simulationSpace=BoneRelative
    boundBone {boneName=dyn_topBow_r}
    componentAppliedLinearAccClamp=<3000 3000 3000>
    componentLinearAccScale=<0.3 0.3 0.6>
    componentLinearVelScale=<0.3 0.3 0.6>
    gravityOverride {z=-980}
    physicsBodyDefinitions [{
      boundBone {boneName=dyn_topBow_r}
      boxExtents=<20 20 20>
      constraintSetup {angularLimitsMax=<0 10 0>, angularLimitsMin=<0 -30 0>, angularTarget=<1 0 0>}
      localJointOffset=<0.45 1 -1>
     }]
    relativeSpaceBone {boneName=pack}
    simSpaceSettings {simSpaceAngularAlpha=0.1}
   }
   n29 = Anim Dynamics - Bone: dyn_bottomBow_r(ComponentPose=@n28, Alpha=1, GravityOverride="0,0,-980", ExternalForce="0,0,0") {
    angularDampingOverride=0.999
    angularSpringConstant=3000
    bAngularSpring=true
    bOverrideAngularDamping=true
    bUseGravityOverride=true
    gravityScale=0.35
    simulationSpace=BoneRelative
    boundBone {boneName=dyn_bottomBow_r}
    componentAppliedLinearAccClamp=<3000 3000 3000>
    componentLinearAccScale=<0.3 0.3 0.5>
    componentLinearVelScale=<0.3 0.3 0.5>
    gravityOverride {z=-980}
    physicsBodyDefinitions [{
      boundBone {boneName=dyn_bottomBow_r}
      boxExtents=<18 18 18>
      constraintSetup {angularLimitsMax=<0 4 0>, angularLimitsMin=<0 -20 0>, angularTarget=<1 0 0>}
      localJointOffset=<0.6 0 0>
     }]
    relativeSpaceBone {boneName=pack}
    simSpaceSettings {simSpaceAngularAlpha=0.1}
   }
   n30 = Anim Dynamics - Bone: dyn_topStrings_r(ComponentPose=@n29, GravityOverride="0,0,-980", ExternalForce="0,0,0") {
    alpha=0.7
    angularDampingOverride=0.999
    angularSpringConstant=3000
    bAngularSpring=true
    bOverrideAngularDamping=true
    bUseGravityOverride=true
    gravityScale=0.35
    simulationSpace=BoneRelative
    boundBone {boneName=dyn_topStrings_r}
    componentAppliedLinearAccClamp=<3000 3000 3000>
    componentLinearAccScale=<0.3 0.3 0.5>
    componentLinearVelScale=<0.3 0.3 0.5>
    gravityOverride {z=-980}
    physicsBodyDefinitions [{
      boundBone {boneName=dyn_topStrings_r}
      boxExtents=<11 11 11>
      constraintSetup {angularLimitsMax=<0 10 0>, angularLimitsMin=<0 -20 0>, angularTarget=<1 0 0>}
      localJointOffset=<0.2 0 0.3>
     }]
    relativeSpaceBone {boneName=pack}
    simSpaceSettings {simSpaceAngularAlpha=0.1}
   }
   n31 = Anim Dynamics - Bone: dyn_topStringsEnd_r(ComponentPose=@n30, GravityOverride="0,0,-980", ExternalForce="0,0,0") {
    alpha=0.7
    angularDampingOverride=0.999
    angularSpringConstant=2000
    bAngularSpring=true
    bOverrideAngularDamping=true
    bUseGravityOverride=true
    gravityScale=0.35
    simulationSpace=BoneRelative
    boundBone {boneName=dyn_topStringsEnd_r}
    componentAppliedLinearAccClamp=<3000 3000 3000>
    componentLinearAccScale=<0.05 0.03 0.5>
    componentLinearVelScale=<0.05 0.01 0.5>
    gravityOverride {z=-980}
    physicsBodyDefinitions [{
      boundBone {boneName=dyn_topStringsEnd_r}
      boxExtents=<28 28 28>
      constraintSetup {angularLimitsMax=<0 5 15>, angularLimitsMin=<0 -5 0>, angularTarget=<1 0 0>}
      localJointOffset=<1.5 -1 0.6>
     }]
    relativeSpaceBone {boneName=pack}
    simSpaceSettings {simSpaceAngularAlpha=0.1}
   }
   n32 = Anim Dynamics - Bone: dyn_topStringsSml_r(ComponentPose=@n31, GravityOverride="0,0,-980", ExternalForce="0,0,0") {
    alpha=0.7
    angularDampingOverride=0.999
    angularSpringConstant=3200
    bAngularSpring=true
    bOverrideAngularDamping=true
    bUseGravityOverride=true
    gravityScale=0.35
    simulationSpace=BoneRelative
    boundBone {boneName=dyn_topStringsSml_r}
    componentAppliedLinearAccClamp=<3000 3000 3000>
    componentLinearAccScale=<0.3 0.3 0.5>
    componentLinearVelScale=<0.3 0.3 0.5>
    gravityOverride {z=-980}
    physicsBodyDefinitions [{
      boundBone {boneName=dyn_topStringsSml_r}
      boxExtents=<13 13 13>
      constraintSetup {angularLimitsMax=<0 10 10>, angularLimitsMin=<0 -20 0>, angularTarget=<1 0 0>}
      localJointOffset=<0.3 -1 0.3>
     }]
    relativeSpaceBone {boneName=pack}
    simSpaceSettings {simSpaceAngularAlpha=0.1}
   }
   n33 = Anim Dynamics - Bone: dyn_topBow_l(ComponentPose=@n32, Alpha=1, GravityOverride="0,0,-980", ExternalForce="0,0,0") {
    angularDampingOverride=0.999
    angularSpringConstant=3000
    bAngularSpring=true
    bOverrideAngularDamping=true
    bUseGravityOverride=true
    gravityScale=0.35
    simulationSpace=BoneRelative
    boundBone {boneName=dyn_topBow_l}
    componentAppliedLinearAccClamp=<3000 3000 3000>
    componentLinearAccScale=<0.3 0.3 0.6>
    componentLinearVelScale=<0.3 0.3 0.6>
    gravityOverride {z=-980}
    physicsBodyDefinitions [{
      boundBone {boneName=dyn_topBow_l}
      boxExtents=<20 20 20>
      constraintSetup {angularLimitsMax=<0 10 0>, angularLimitsMin=<0 -30 0>, angularTarget=<1 0 0>}
      localJointOffset=<0.3 0 1>
     }]
    relativeSpaceBone {boneName=pack}
    simSpaceSettings {simSpaceAngularAlpha=0.1}
   }
   n34 = Anim Dynamics - Bone: dyn_bottomBow_l(ComponentPose=@n33, Alpha=1, GravityOverride="0,0,-980", ExternalForce="0,0,0") {
    angularDampingOverride=0.999
    angularSpringConstant=3000
    bAngularSpring=true
    bOverrideAngularDamping=true
    bUseGravityOverride=true
    gravityScale=0.35
    simulationSpace=BoneRelative
    boundBone {boneName=dyn_bottomBow_l}
    componentAppliedLinearAccClamp=<3000 3000 3000>
    componentLinearAccScale=<0.3 0.3 0.5>
    componentLinearVelScale=<0.3 0.3 0.5>
    gravityOverride {z=-980}
    physicsBodyDefinitions [{
      boundBone {boneName=dyn_bottomBow_l}
      boxExtents=<18 18 18>
      constraintSetup {angularLimitsMax=<0 10 0>, angularLimitsMin=<0 -20 0>, angularTarget=<1 0 0>}
      localJointOffset=<0.35 0 0>
     }]
    relativeSpaceBone {boneName=pack}
    simSpaceSettings {simSpaceAngularAlpha=0.1}
   }
   n35 = Anim Dynamics - Bone: dyn_topStrings_l(ComponentPose=@n34, GravityOverride="0,0,-980", ExternalForce="0,0,0") {
    alpha=0.7
    angularDampingOverride=0.95
    angularSpringConstant=900
    bAngularSpring=true
    bOverrideAngularDamping=true
    bUseGravityOverride=true
    gravityScale=0.35
    simulationSpace=BoneRelative
    boundBone {boneName=dyn_topStrings_l}
    componentAppliedLinearAccClamp=<3000 3000 3000>
    componentLinearAccScale=<0.3 0.3 0.4>
    componentLinearVelScale=<0.3 0.3 0.4>
    gravityOverride {z=-980}
    physicsBodyDefinitions [{
      boundBone {boneName=dyn_topStrings_l}
      boxExtents=<12 12 12>
      constraintSetup {angularLimitsMax=<0 10 0>, angularLimitsMin=<0 -20 0>, angularTarget=<1 0 0>}
      localJointOffset=<0.1 0 0.3>
     }]
    relativeSpaceBone {boneName=pack}
    simSpaceSettings {simSpaceAngularAlpha=0.1}
   }
   n36 = Anim Dynamics - Bone: dyn_topStringsEnd_l(ComponentPose=@n35, GravityOverride="0,0,-980", ExternalForce="0,0,0") {
    alpha=0.7
    angularDampingOverride=0.999
    angularSpringConstant=1500
    bAngularSpring=true
    bOverrideAngularDamping=true
    bUseGravityOverride=true
    gravityScale=0.35
    simulationSpace=BoneRelative
    boundBone {boneName=dyn_topStringsEnd_l}
    componentAppliedLinearAccClamp=<3000 3000 3000>
    componentLinearAccScale=<0.05 0.03 0.5>
    componentLinearVelScale=<0.05 0.01 0.5>
    gravityOverride {z=-980}
    physicsBodyDefinitions [{
      boundBone {boneName=dyn_topStringsEnd_l}
      boxExtents=<28 28 28>
      constraintSetup {angularLimitsMax=<0 5 30>, angularLimitsMin=<0 -5 0>, angularTarget=<1 0 0>}
      localJointOffset=<1.5 -1.5 0.6>
     }]
    relativeSpaceBone {boneName=pack}
    simSpaceSettings {simSpaceAngularAlpha=0.1}
   }
   n37 = Anim Dynamics - Bone: dyn_topStringsSml_l(ComponentPose=@n36, GravityOverride="0,0,-980", ExternalForce="0,0,0") {
    alpha=0.7
    angularDampingOverride=0.95
    angularSpringConstant=700
    bAngularSpring=true
    bOverrideAngularDamping=true
    bUseGravityOverride=true
    gravityScale=0.35
    simulationSpace=BoneRelative
    boundBone {boneName=dyn_topStringsSml_l}
    componentAppliedLinearAccClamp=<3000 3000 3000>
    componentLinearAccScale=<0.3 0.3 0.4>
    componentLinearVelScale=<0.3 0.3 0.4>
    gravityOverride {z=-980}
    physicsBodyDefinitions [{
      boundBone {boneName=dyn_topStringsSml_l}
      boxExtents=<12 12 12>
      constraintSetup {angularLimitsMax=<0 10 0>, angularLimitsMin=<0 -20 0>, angularTarget=<1 0 0>}
      localJointOffset=<0.1 0 0.3>
     }]
    relativeSpaceBone {boneName=pack}
    simSpaceSettings {simSpaceAngularAlpha=0.1}
   }
   n38 = Get bIsMoving2D()
   n39 = "Transform (Modify) Bone - Bone: dyn_chain_02_l"!(ComponentPose=@n37, bAlphaBoolEnabled:bool=@n38, Translation="0,0,0", Rotation="0,6,0", Scale="1,1,1") -> (Pose) {
    alphaInputType=Bool
    rotationMode=BMM_Additive
    rotationSpace=BCS_BoneSpace
    translationMode=BMM_Additive
    translationSpace=BCS_BoneSpace
    alphaBoolBlend {blendInTime=0.2, blendOutTime=0.3}
    boneToModify {boneName=dyn_chain_02_l}
    rotation {yaw=6}
   }
   n40 = "Transform (Modify) Bone - Bone: dyn_chain_02_r"!(ComponentPose=@n39, bAlphaBoolEnabled:bool=@n38, Translation="0,0,0", Rotation="0,6,0", Scale="1,1,1") -> (Pose) {
    alphaInputType=Bool
    rotationMode=BMM_Additive
    rotationSpace=BCS_BoneSpace
    translationMode=BMM_Additive
    translationSpace=BCS_BoneSpace
    alphaBoolBlend {blendInTime=0.2, blendOutTime=0.3}
    boneToModify {boneName=dyn_chain_02_r}
    rotation {yaw=6}
   }
   n41 = Get bIsSkyDiving()
   n42 = "Transform (Modify) Bone - Bone: dyn_bottomBow_r"!(ComponentPose=@n40, bAlphaBoolEnabled:bool=@n41, Translation="0,0,0", Rotation="7.350,-0.000,0.000", Scale="1,1,1") -> (Pose) {
    alphaInputType=Bool
    rotationMode=BMM_Additive
    rotationSpace=BCS_BoneSpace
    translationSpace=BCS_BoneSpace
    alphaBoolBlend {blendInTime=0.2, blendOutTime=0.3}
    boneToModify {boneName=dyn_bottomBow_r}
    rotation=<rot 7.35 -0.0001 0>
   }
   n43 = "Transform (Modify) Bone - Bone: dyn_bottomBow_l"!(ComponentPose=@n42, bAlphaBoolEnabled:bool=@n41, Translation="0,0,0", Rotation="-5.414,0.755,2.160", Scale="1,1,1") -> (Pose) {
    alphaInputType=Bool
    rotationMode=BMM_Additive
    rotationSpace=BCS_BoneSpace
    translationSpace=BCS_BoneSpace
    alphaBoolBlend {blendInTime=0.2, blendOutTime=0.3}
    boneToModify {boneName=dyn_bottomBow_l}
    rotation=<rot -5.4141 0.7548 2.16>
   }
   n44 = Get bIsDBNO()
   n45 = "Transform (Modify) Bone - Bone: dyn_chain_02_l"(ComponentPose=@n43, bAlphaBoolEnabled=@n44, Translation="0,0,0", Rotation="0,6,0", Scale="1,1,1") {
    alphaInputType=Bool
    rotationMode=BMM_Additive
    rotationSpace=BCS_BoneSpace
    translationMode=BMM_Additive
    translationSpace=BCS_BoneSpace
    alphaBoolBlend {blendInTime=0.2, blendOutTime=0.3}
    boneToModify {boneName=dyn_chain_02_l}
    rotation {yaw=6}
   }
   n46 = "Transform (Modify) Bone - Bone: dyn_chain_02_r"(ComponentPose=@n45, bAlphaBoolEnabled=@n44, Translation="0,0,0", Rotation="0,6,0", Scale="1,1,1") {
    alphaInputType=Bool
    rotationMode=BMM_Additive
    rotationSpace=BCS_BoneSpace
    translationMode=BMM_Additive
    translationSpace=BCS_BoneSpace
    alphaBoolBlend {blendInTime=0.2, blendOutTime=0.3}
    boneToModify {boneName=dyn_chain_02_r}
    rotation {yaw=6}
   }
   n47 = Component To Local(ComponentPose=@n46)
   n48 = Get bIsInFrontEnd()
   n49 = "Transform (Modify) Bone - Bone: dyn_bottomBow_r"(
    ComponentPose=Local To Component(LocalPose=Blend Poses by bool(BlendPose_0=@n26, BlendPose_1=@n47, bActiveValue=@n48))
    bAlphaBoolEnabled=True
    Translation="-0.031,0.199,0.300"
    Rotation="4.433,-0.027,-2.464"
    Scale="1,1,1"
   ) {
    alphaInputType=Bool
    rotationMode=BMM_Additive
    rotationSpace=BCS_BoneSpace
    translationMode=BMM_Additive
    translationSpace=BCS_BoneSpace
    alphaBoolBlend {blendInTime=0.2, blendOutTime=0.3}
    boneToModify {boneName=dyn_bottomBow_r}
    rotation=<rot 4.433 -0.0271 -2.4639>
    translation=<-0.031 0.1987 0.3>
   }
   n50 = "Transform (Modify) Bone - Bone: body"(ComponentPose=@n49, Alpha=1, Translation="0,3.8,-2.645", Rotation="0,2.5,0", Scale="1,1,1") {
    rotationMode=BMM_Additive
    translationMode=BMM_Additive
    translationSpace=BCS_BoneSpace
    boneToModify {boneName=body}
    rotation {yaw=2.5}
    translation {y=3.8, z=-2.645}
   }
   n51 = "Transform (Modify) Bone - Bone: pack"(ComponentPose=@n50, Translation=Get backpack_translation(), Rotation=Get backpack_rotate()) {
    rotationMode=BMM_Additive
    rotationSpace=BCS_BoneSpace
    translationMode=BMM_Additive
    translationSpace=BCS_BoneSpace
    alphaBoolBlend {blendInTime=0.1, blendOption=Linear}
    boneToModify {boneName=pack}
   }
   Output Pose(Result=Component To Local(ComponentPose=@n51)) {layerGroup=None}
   n52 = Anim Dynamics - Bone: dyn_bambooStick!(
    ComponentPose=Local To Component(LocalPose=Input Pose())
    bAlphaBoolEnabled:bool=Get bambooSwitch() -> (bambooSwitch:bool)
    GravityOverride:Vector="0,0,-980"
    ExternalForce="0,0,0"
   ) -> (Pose) {
    alphaInputType=Bool
    angularDampingOverride=0.999
    angularSpringConstant=2000
    bAngularSpring=true
    bOverrideAngularDamping=true
    bUseGravityOverride=true
    gravityScale=0.35
    simulationSpace=BoneRelative
    boundBone {boneName=dyn_bambooStick}
    componentAppliedLinearAccClamp=<3000 3000 3000>
    componentLinearAccScale=<0.3 0.3 0.6>
    componentLinearVelScale=<0.3 0.3 0.6>
    gravityOverride {z=-980}
    physicsBodyDefinitions [{
      boundBone {boneName=dyn_bambooStick}
      boxExtents=<40 40 40>
      constraintSetup {angularLimitsMax=<0 20 0>, angularLimitsMin=<0 -20 0>, angularTarget=<1 0 0>}
      localJointOffset=<3 0 -0.2>
     }]
    relativeSpaceBone {boneName=pack}
    simSpaceSettings {simSpaceAngularAlpha=0.1}
   }
   Save cached pose 'chainRBAN'(Pose=Component To Local(ComponentPose=@n52)) {cacheName=chainRBAN}
  ]}
 EventGraph {nodes [
   Event BlueprintUpdateAnimation()
   > n1 = Sequence!(execute) -> (then_0, then_1, then_2:Exec)
   n2 = Get bIsSkydiveFloating()
   n3 = Get bIsOnGrindRail()
   Set bambooSwitch(execute=@n1.then_2, bambooSwitch:bool=NOT Boolean(A=OR Boolean(A=@n3, B=@n2))) -> (Output_Get:bool)
   n4 = GetOwnerSkeletalMeshComponent()
   n5 = BreakRotator(InRot=BreakTransform(InTransform=GetDeltaTransformFromRefPose(self=@n4, BoneName=spine_05, BaseName=spine_01)).Rotation)
   Set backpack_rotate(execute=@n1.then_1, backpack_rotate=MakeRotator(Roll=MapRangeClamped(Value=@n5.Yaw, InRangeA=0, InRangeB=-20, OutRangeB=15)))
   > Set backpack_translation(backpack_translation=MakeVector(Y=MapRangeClamped(Value=@n5.Yaw, InRangeA=0, InRangeB=-18, OutRangeB=-4)))
   n6 = "Gravity Override State Function (Parameter Asset) (Nativized)"(
    execute=@n1.then_0
    CharacterPartAnimInstance=Self-Reference()
    GameTimeInSeconds=GetGameTimeInSeconds()
    DebugJointName=spine_05
    Parameters=&/M_MED_AtticDrill_Pack_Parameters
   )
   > Set Gravity Override(Gravity Override=@n6)
   TryGetPawnOwner()
  ]}
]
variables [
 backpack_rotate {type=Rotator}
 backpack_translation {type=Vector}
 bambooSwitch {type=boolean}
 Gravity Override {type=Vector}
]
